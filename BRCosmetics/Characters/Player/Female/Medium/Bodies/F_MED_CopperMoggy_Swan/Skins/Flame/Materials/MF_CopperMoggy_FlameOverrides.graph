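# Material function for the CopperMoggy Flame skin that adds a colored, masked halftone dot pattern to the incoming material's emissive.
# A dotted distance-field texture tiles over stabilized screen-space UVs anchored to the primitive or actor position plus an offset.
# A gradient varies dot size, and screen derivatives keep dot edges antialiased.
# It also declares an invertible TechArtMask effect mask and a dilated dot mask, and outputs a Temporal Mask.
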MaterialFunction /BRCosmetics/Characters/Player/Female/Medium/Bodies/F_MED_CopperMoggy_Swan/Skins/Flame/Materials/MF_CopperMoggy_FlameOverrides
graph {nodes [
  n1 = GetMaterialAttributes(None=FunctionInput() {inputName=Material, inputType=FunctionInput_MaterialAttributes}) {attributeGetTypes [#1]}
  n2 = LinearInterpolate(
   A=TransformPosition(None=Constant3Vector()) {transformType=TRANSFORMPOSSOURCE_World}
   B=ActorPositionWS()
   Alpha=ScalarParameter() {defaultValue=1, desc="0 = Primitive Position, 1 = Actor Position", group=Flame, parameterName=UseActorPosition}
  )
  n3 = VectorParameter() {group=Flame, parameterName=CenterOffset, defaultValue=<rgba 0 0 150 1>}
  n4 = MaterialFunctionCall!(ObjectPivotLocation=@n2, ObjectPositionOffset=@n3.RGB) {materialFunction=/Game/Characters/Player/Common/MaterialFunctions/Superlevel/ImprovedStabilizedScreenSpaceUVs}
  n5 = ScalarParameter() {defaultValue=80, group=Flame, parameterName=FlamePatternTiling}
  n6 = TextureSample(UVs=Multiply(
    A=NamedRerouteDeclaration(None=@n4.StabilizedUVs) {name=ScreenSpaceUVs, nodeColor=<rgba 0.7109 1 0 1>}
    B=NamedRerouteDeclaration(None=@n5) {name=PatternTIling, nodeColor=<rgba 0.5234 1 0 1>}
   )) {
   samplerSource=SSM_Wrap_WorldGroupSettings
   samplerType=SAMPLERTYPE_LinearGrayscale
   texture=/BRCosmetics/Weapons/FORT_Melee/Pickaxe_ScoopPluck/Textures/T_DottedPattern_DF
  }
  n7 = Subtract(A=ScalarParameter() {defaultValue=0.1, group=Flame, parameterName=MinDotSize})
  n8 = Subtract(A=ScalarParameter() {defaultValue=1.3, group=Flame, parameterName=MaxDotSize})
  n9 = Saturate(None=ComponentMask(None=NamedRerouteUsage()) {g=true})
  n10 = Saturate(None=Add(
    A=NamedRerouteDeclaration(None=@n6.R) {name=PatternDistanceField, nodeColor=<rgba 1 0 0.3984 1>}
    B=LinearInterpolate(A=@n7, B=@n8, Alpha=@n9) {constA=-1, constAlpha=1, constB=0.3}
   ))
  n11 = Abs(None=DDX(Value=@n10))
  n12 = Abs(None=DDY(Value=@n10))
  n13 = Divide(A=Subtract(A=@n10) {constB=0.5}, B=Max(A=Add(A=@n11, B=@n12)) {constB=0})
  n14 = Add(A=@n13) {constB=0.5}
  n15 = Multiply(
   A=Multiply(A=Saturate(None=@n14), B=VectorParameter() {group=Flame, parameterName=PatternColor, defaultValue=<rgba 0.9131 0.7835 0.0908 1>}.RGB)
   B=NamedRerouteUsage()
  )
  FunctionOutput(None=SetMaterialAttributes!(MaterialAttributes=@n1.MaterialAttributes, EmissiveColor=Add(A=@n1.EmissiveColor, B=@n15)) {attributeSetTypes [#1]})
  FunctionOutput(None=Multiply(A=NamedRerouteUsage(), B=NamedRerouteUsage())) {outputName=Temporal Mask, sortPriority=1}
  LocalPosition()
  MaterialFunctionCall!(
   Position=NamedRerouteDeclaration(None=Add(A=@n2, B=@n3.RGB)) {name=ObjectPosition, nodeColor=<rgba 0.4766 1 0 1>}
   UVRadialScale=Constant() {r=1}
  ) {materialFunction=/Game/Weapons/WeaponSkins/MaterialFunction/ObjectStabilizedScreenSpaceUVs}
  n16 = StaticComponentMaskParameter(None=FunctionInput() {inputName=TechArtMask, sortPriority=1}) {defaultG=true, group=Flame, parameterName=MaskComponent}
  NamedRerouteDeclaration(None=LinearInterpolate(A=@n16, B=OneMinus(None=@n16), Alpha=ScalarParameter() {group=Flame, parameterName=InvertMask})) {name=EffectMask, nodeColor=<rgba 1 0.3984 0 1>}
  NamedRerouteDeclaration(None=Saturate(None=Add(A=@n13) {constB=5})) {name=DilatedMaskResult, nodeColor=<rgba 1 0.8438 0 1>}
 ]}
